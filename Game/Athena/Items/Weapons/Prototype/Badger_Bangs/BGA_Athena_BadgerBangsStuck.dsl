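; Stuck Badger Bangs proximity mine that beeps near non-teammates, explodes on trigger or damage, and returns an item when picked up.
; Outer sphere runs a timer scaling team-colored blink particles by closest intruder distance; Niagara particle callbacks play beeps.
; Inner sphere explodes with line of sight, otherwise applies a gameplay effect for LOS rechecking.
; Explosions apply an LOS-filtered radial GameplayEffect, delaying 0.25s when another mine is the damage causer.
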(event EventBeginPlay
  (bind _beepparticles (Variables|Default|GetBeepParticles))
  (bind _visiblemesh (Variables|Default|GetVisibleMesh))
  (|CallParentFunction)
  (Utilities|FlowControl|Sequence
    (:then_0
      ;[1 1 1] Caching this so it's not lost when instigator respawns.
      (Variables|Default|SetMyController (Pawn|GetController (Variables|Actor|GetInstigator)))
      (bind _returnvalue (Effects|SpawnFXSystemAttached (Variables|Default|GetEnemyBeepParticlesTemplate) (Variables|Default|GetVisibleMesh)))
      (bind _output_get (Variables|Default|SetBeepParticles _returnvalue))
      (Effects|Components|ParticleSystem|SetColorParameter _output_get "Color" (select (Team|OnSameTeam (Game|GetPlayerController 0) (Variables|Default|GetMyController)) "(R=0.000000,G=0.750000,B=1.500000,A=1.000000)" "(R=1.500000,G=0.025000,B=0.000000,A=1.000000)")))
    (:then_1
      (Components|Sphere|SetSphereRadius (Variables|Default|GetBeepBlinkTrigger) (ScalableFloat|GetValueatLevel (Variables|Default|GetBeepTriggerRadius)) true) ; NOTE: scale of root is set to 2 when BGA is spawned, so this radius is actually doubled.
      (Components|Sphere|SetSphereRadius (Variables|Default|GetExplosionTrigger) (ScalableFloat|GetValueatLevel (Variables|Default|GetTriggerRadius)) true) ; NOTE: scale of root is set to 2 when BGA is spawned, so this radius is actually doubled.
      (if (Variables|Default|GetSetMeshMat)
        (Rendering|Material|SetMaterial (Variables|Default|GetVisibleMesh) 0 "/Game/Athena/Items/Weapons/Prototype/Badger_Bangs/FX/MI_BlackMonday_Prj_Cooldown.MI_BlackMonday_Prj_Cooldown")))
    (:then_2
      (Effects|Components|ParticleSystem|SetBoolParameter _beepparticles "Static" true)
      (Effects|Components|ParticleSystem|SetBoolParameter _beepparticles "Blink" false)
      (Utilities|FlowControl|Sequence
        (:then_0
          (Effects|SpawnEmitteratLocation "/Game/Athena/Items/Weapons/Prototype/Badger_Bangs/FX/P_BlackMonday_HitWorld.P_BlackMonday_HitWorld" (Transformation|GetActorLocation) (Math|Conversions|RotationFromXVector (Variables|Default|GetImpactNormal)))
          (EventDispatchers|BindEventtoOnDied (delegate Custom|Destroyed))
          (Networking|SwitchHasAuthority))
        (:then_1
          (Rendering|Material|SetScalarParameterValueonMaterials _visiblemesh "GameTime" (Utilities|Time|GetGameTimeinSeconds))
          (Rendering|Material|SetScalarParameterValueonMaterials _visiblemesh "AnimDuration" 0.58)
          (Utilities|FlowControl|Delay 0.58)
          (if (Variables|Default|GetSetMeshMat)
            (Rendering|Material|SetMaterial (Variables|Default|GetVisibleMesh) 0 "/Game/Gadgets/Assets/BlackMonday_Bangs/Materials/MI_BlackMonday_Bangs.MI_BlackMonday_Bangs")))))))

(event Custom|Destroyed
  (params (DamagedActor :type Actor%20Object%20Reference) (Damage :type Float%20%28single-precision%29) (InstigatedBy :type Controller%20Object%20Reference) (DamageCauser :type Actor%20Object%20Reference) (HitLocation :type Vector) (FHitComponent :type Primitive%20Component%20Object%20Reference) (BoneName :type Name) (Momentum :type Vector))
  (Audio|Components|Audio|FadeOut (Variables|Default|GetAmbientHumAudio) 0.5 0.0 "Linear" :!class /Script/Engine.AudioComponent)
  (if (not (Variables|Default|GetIsExlode?))
    (Effects|Components|ParticleSystem|SpawnFXAtLocation "/Game/Athena/Items/Weapons/Prototype/Badger_Bangs/FX/NS_BlackMonday_PrjTakeBack.NS_BlackMonday_PrjTakeBack" (Transformation|GetActorLocation) (Math|Conversions|RotationFromXVector (Variables|Default|GetImpactNormal)))
    (Audio|PlaySoundatLocation (Variables|Default|GetPickUpSound) (Transformation|GetActorLocation)))
  (Rendering|SetActorHiddenInGame true)
  (Collision|SetActorEnableCollision false))

(event OnComponentBeginOverlap(BeepBlinkTrigger)
  (params (OverlappedComponent :type Primitive%20Component%20Object%20Reference) (OtherActor :type Actor%20Object%20Reference) (OtherComp :type Primitive%20Component%20Object%20Reference) (OtherBodyIndex :type Integer) (bFromSweep :type Boolean) (SweepResult :type Hit%20Result%20Structure%20%28by%20ref%29))
  ;[1 0.5749 0.505] Beep Blink FX
  (Networking|SwitchHasAuthority
    (:Authority
      (CallFunction|TriggerBeep OtherActor))))

(event Niagara|EventReceiveParticleData
  (params (Data :type Array%20of%20Basic%20Particle%20Data%20Structures) (NiagaraSystem :type Niagara%20System%20Object%20Reference) (SimulationPositionOffset :type Vector%20%28by%20ref%29 :default "0, 0, 0"))
  ;[1 0.5749 0.505] Beep Blink FX
  (for _array_element Data
    (Audio|PlaySoundatLocation (Variables|Default|GetBeepSound) (Transformation|GetActorLocation))
    (Fortnite|ShowSoundIndicatorLocal :Location (Transformation|GetActorLocation) :MaxAudibleDistance (Variables|/Script/Engine.SoundBase|GetMaxDistance (Variables|Default|GetBeepSound)) :IndicatorType "Generic" :Instigator self :Tint "(R=0.000000,G=0.000000,B=0.000000,A=0.000000)" :OverrideIcon (Variables|Default|GetSoundIndicatorIcon) :AttenuationMaxRangeMultiplier 1.0 :AttenuationIntensityMultiplier 1.0)))

(event Custom|TriggerBeep
  (params (ThePlayerPawnWhoTriggers :type Actor%20Object%20Reference))
  (bind _beepparticles (Variables|Default|GetBeepParticles))
  ;[1 0.5749 0.505] Beep Blink FX
  (if (not (Team|OnSameTeam (Utilities|Casting|CastToFortPlayerPawnAthena ThePlayerPawnWhoTriggers) (Variables|Default|GetMyController)))
    (Utilities|Array|AddUnique (Variables|Default|GetActorsInBeepRadius) ThePlayerPawnWhoTriggers)
    (if (not (Utilities|Time|IsValidTimerHandle (Variables|Default|GetBeepTimer)))
      (bind _returnvalue (Utilities|Time|SetTimerbyEvent (EventDispatchers|CreateEvent :!event Custom|PlayBeepFX) 0.05 true))
      (Variables|Default|SetBeepTimer _returnvalue)
      (Utilities|IsValid _beepparticles
        (:Is%20Valid
          (bind _asniagara_particle_system_component (Utilities|Casting|CastToNiagaraComponent _beepparticles))
          (Niagara|SetNiagaraVariable(Object) _asniagara_particle_system_component "BPData" self)
          (Effects|Components|ParticleSystem|SetBoolParameter _beepparticles "Static" false)
          (Effects|Components|ParticleSystem|SetBoolParameter _beepparticles "Blink" true))))))

(event Custom|Beep
  (params (EventName :type Name) (EmitterTime :type Float%20%28single-precision%29) (Location :type Vector) (Velocity :type Vector))
  ;[1 0.5749 0.505] Beep Blink FX
  (if (Utilities|Name|Equal(Name) EventName "Beep")
    _))

(event OnComponentEndOverlap(BeepBlinkTrigger)
  (params (OverlappedComponent :type Primitive%20Component%20Object%20Reference) (OtherActor :type Actor%20Object%20Reference) (OtherComp :type Primitive%20Component%20Object%20Reference) (OtherBodyIndex :type Integer))
  ;[1 0.5749 0.505] Beep Blink FX
  (Networking|SwitchHasAuthority
    (:Authority
      (Utilities|Array|RemoveItem (Variables|Default|GetActorsInBeepRadius) OtherActor)
      (if (== (Utilities|Array|Length (Variables|Default|GetActorsInBeepRadius)) 0)
        (CallFunction|StopBeep)))))

(event Custom|PlayBeepFX
  (bind _actorsinbeepradius (Variables|Default|GetActorsInBeepRadius))
  (bind _self self)
  (bind _returnvalue (Fortnite|GetClosestActorfromArray _self _actorsinbeepradius))
  (bind _beepblinktrigger (Variables|Default|GetBeepBlinkTrigger))
  (bind _returnvalue_2 (Components|Sphere|GetScaledSphereRadius _beepblinktrigger))
  (bind _returnvalue_3 (Transformation|GetActorLocation _returnvalue))
  (bind _returnvalue_4 (Transformation|GetActorLocation))
  (bind _returnvalue_5 (Math|Vector|Distance(Vector) _returnvalue_3 _returnvalue_4))
  (bind _returnvalue_6 (Math|Float|NormalizetoRange _returnvalue_5 0.0 _returnvalue_2))
  (bind _beepparticles (Variables|Default|GetBeepParticles))
  ;[1 0.5749 0.505] Beep Blink FX
  (Utilities|IsValid _returnvalue
    (:Is%20Valid
      (Effects|Components|ParticleSystem|SetFloatParameter _beepparticles "Spawn Rate" (Math|Float|Lerp 7.0 0.5 _returnvalue_6))
      (Effects|Components|ParticleSystem|SetFloatParameter _beepparticles "Lifetime" (Math|Float|Lerp 0.05 0.35 _returnvalue_6)))
    (:Is%20Not%20Valid
      (CallFunction|StopBeep))))

(event OnComponentBeginOverlap(ExplosionTrigger)
  (params (OverlappedComponent :type Primitive%20Component%20Object%20Reference) (OtherActor :type Actor%20Object%20Reference) (OtherComp :type Primitive%20Component%20Object%20Reference) (OtherBodyIndex :type Integer) (bFromSweep :type Boolean) (SweepResult :type Hit%20Result%20Structure%20%28by%20ref%29))
  (Networking|SwitchHasAuthority
    (:Authority
      (Utilities|CheckforExplosionTrigger OtherActor
        (:Trigger
          (Utilities|CheckLOS OtherActor
            (:LOS
              (CallFunction|Explode))
            (:No%20LOS
              (GameplayEffects|ApplyGameplayEffectToTarget :self (Ability|GetAbilitySystemComponent self) :GameplayEffectClass "/Game/Athena/Items/Weapons/Prototype/Badger_Bangs/GE_BadgerBangs_ApplyLOSChecker.GE_BadgerBangs_ApplyLOSChecker_C" :Target (Ability|GetAbilitySystemComponent OtherActor) :Level 0.0))))))))

(event Custom|StopBeep
  (bind _beepparticles (Variables|Default|GetBeepParticles))
  ;[1 0.5749 0.505] Beep Blink FX
  (Utilities|Time|ClearandInvalidateTimerbyHandle (Variables|Default|GetBeepTimer))
  (Utilities|IsValid _beepparticles
    (:Is%20Valid
      (Effects|Components|ParticleSystem|SetBoolParameter _beepparticles "Static" true)
      (Effects|Components|ParticleSystem|SetBoolParameter _beepparticles "Blink" false))))

(event OnComponentEndOverlap(ExplosionTrigger)
  (params (OverlappedComponent :type Primitive%20Component%20Object%20Reference) (OtherActor :type Actor%20Object%20Reference) (OtherComp :type Primitive%20Component%20Object%20Reference) (OtherBodyIndex :type Integer))
  (GameplayEffects|RemoveActiveEffectswithTags (Ability|GetAbilitySystemComponent OtherActor) "(GameplayTags=((TagName=\"Item.Consumable.BadgerBangs\")))"))

(event EventOnDamageServer
  (params (Damage :type Float%20%28single-precision%29) (DamageTags :type Gameplay%20Tag%20Container%20Structure%20%28by%20ref%29) (Momentum :type Vector :default "0, 0, 0") (HitInfo :type Hit%20Result%20Structure%20%28by%20ref%29) (InstigatedBy :type Controller%20Object%20Reference) (DamageCauser :type Actor%20Object%20Reference) (EffectContext :type Gameplay%20Effect%20Context%20Handle%20Structure))
  (Utilities|Casting|CastToBGA_Athena_BadgerBangsStuck DamageCauser
    (:then
      (Utilities|FlowControl|Delay 0.25))
    (:CastFailed
      _))
  (CallFunction|Explode))

(event Custom|Explode
  (bind _returnvalue (Transformation|GetActorLocation))
  (bind _self self)
  (if (not (Variables|Default|GetIsExlode?))
    (Variables|Default|SetIsExlode? true)
    (Utilities|FlowControl|Sequence
      (:then_0
        (Collision|SetActorEnableCollision false)
        (Actor|SetLifeSpan 0.2)
        (Audio|PlaySoundatLocation (Variables|Default|GetExplodeSound) _returnvalue "0, 0, 0" 1.0 1.0 0.0 0 0 self)
        (Effects|SpawnEmitteratLocation (Variables|Default|GetExplosionEffect) _returnvalue (Math|Conversions|RotationFromXVector (Variables|Default|GetImpactNormal)))
        (GameplayCue|ExecuteGameplayCue%28Burst%29Local self (Variables|Default|GetCameraShakeGC) (Ability|GameplayCue|MakeGameplayCueParameters :NormalizedMagnitude 1.0 :RawMagnitude 3.0 :Location _returnvalue :Normal "0,0,0" :GameplayEffectLevel 1 :AbilityLevel 1 :bReplicateLocationWhenUsingMinimalRepProxy false))
        (Fortnite|ShowSoundIndicatorLocalforTeam :Location (Transformation|GetActorLocation) :MaxAudibleDistance (Variables|/Script/Engine.SoundBase|GetMaxDistance (Variables|Default|GetExplodeSound)) :IndicatorType "Gunshot" :Team (Team|GetActorTeam self) :Affiliations (Utilities|Array|MakeArray "Hostile" "Neutral") :Tint "(R=0.000000,G=0.000000,B=0.000000,A=0.000000)" :AttenuationMaxRangeMultiplier 1.0 :AttenuationIntensityMultiplier 1.0))
      (:then_1
        (bind _outactors (Collision|SphereOverlapActors (Transformation|GetActorLocation) (ScalableFloat|GetValueatLevel (Variables|Default|GetDamageRadius)) (Utilities|Array|MakeArray "ObjectTypeQuery6" "ObjectTypeQuery1" "ObjectTypeQuery5" "ObjectTypeQuery3" "ObjectTypeQuery2" "ObjectTypeQuery4" "ObjectTypeQuery7") "/Script/Engine.Actor" (Utilities|Array|MakeArray self)))
        (bind _result (CallFunction|FilterbyLOS _outactors))
        (for _array_element _result
          (bind _asfort_ability_system_component (Utilities|Casting|CastToFortAbilitySystemComponent (Ability|GetAbilitySystemComponent _self)))
          (GameplayEffects|ApplyGameplayEffectToTarget _asfort_ability_system_component (Variables|Default|GetExplosionGE) (Ability|GetAbilitySystemComponent _array_element) 0.0 (GameplayEffects|MakeFortEffectContext _asfort_ability_system_component 0 0 _self 0)))))))

(event Interaction|EventBlueprintOnInteract
  (params (InteractingPawn :type Fort%20Pawn%20Object%20Reference) (InteractionBeingAttempted :type EInteractionBeingAttempted%20Enum :default "FirstInteraction"))
  (bind _asfort_player_controller (Utilities|Casting|CastToFortPlayerController (Pawn|GetController InteractingPawn)))
  ;[1 0 0] IMPORTANT: If this item is unvaulted, ItemDefinition needs to be set to the WID.
  (Inventory|GiveItemtoInventoryOwner _asfort_player_controller (Variables|Default|GetItemToPickUpOnDisarm))
  (Variables|Default|SetIsInteract? true)
  (Actor|DestroyActor))

(event EventSpawnFXSounds
  (|CallParentFunction)
  (Audio|Components|Audio|FadeIn (Variables|Default|GetAmbientHumAudio) 0.5 1.0 0.0 "Linear" :!class /Script/Engine.AudioComponent))

(fn BlueprintCanInteract
  (params (InteractingPawn :type Fort%20Pawn%20Object%20Reference) (InteractionBeingAttempted :type EInteractionBeingAttempted%20Enum :default "FirstInteraction") (InteractionType :type TInteractionType%20Enum :default "IT_NoInteraction"))
  (return true))

(fn BlueprintGetInteractionString
  (params (InteractingPawn :type Fort%20Pawn%20Object%20Reference) (InteractionBeingAttempted :type EInteractionBeingAttempted%20Enum :default "FirstInteraction"))
  (return (Variables|Default|GetInteractText)))

(fn BlueprintGetInteractionTime
  (params (InteractingPawn :type Fort%20Pawn%20Object%20Reference) (InteractionBeingAttempted :type EInteractionBeingAttempted%20Enum :default "FirstInteraction"))
  (return 0.0 true))

(fn CheckLOSAgain
  (params (ActorToCheckLOSAgainst :type Actor%20Object%20Reference))
  (Networking|SwitchHasAuthority
    (:Authority
      (Utilities|CheckLOS ActorToCheckLOSAgainst
        (:LOS
          (if (Variables|Default|GetArmed)
            (CallFunction|Explode)))))))

(fn FilterByLOS
  (params (_Array :type Array%20of%20Actor%20Object%20References))
  (for _array_element _Array
    (Utilities|CheckLOS _array_element
      (:LOS
        (Utilities|Array|Add (LocalVariables|GetResult) _array_element))))
  (return (LocalVariables|GetResult)))

(fn Launch
  (params (LaunchSource :type Vector) (Amount :type Float%20%28double-precision%29))
  (bind _returnvalue (Transformation|GetActorLocation))
  (bind _returnvalue_2 (- _returnvalue LaunchSource))
  (bind _gravhitnormal (Variables|Default|GetGravHitNormal))
  (bind _returnvalue_3 (* _gravhitnormal Amount))
  (bind _returnvalue_4 (Math|Vector|Normalize _returnvalue_2))
  (bind _returnvalue_5 (* _returnvalue_4 Amount))
  (bind _returnvalue_6 (+ _returnvalue_5 _returnvalue_3))
  (bind _returnvalue_7 (Math|Vector|Normalize _returnvalue_6))
  (bind _returnvalue_8 (* _returnvalue_7 1e+02))
  (bind _returnvalue_9 (Transformation|GetActorLocation))
  (bind _returnvalue_10 (+ _returnvalue_9 _returnvalue_8))
  (CallFunction|RestartSimulation)
  (Transformation|DetachFromActor "KeepWorld" "KeepWorld" "KeepWorld")
  (bind (_outhit _returnvalue_11) (Collision|LineTraceByChannel _returnvalue_9 _returnvalue_10 "TraceTypeQuery10"))
  (Development|PrintString (Utilities|String|ToString(Vector) _returnvalue_8) true true "(R=1.000000,G=0.000000,B=0.847731,A=1.000000)" 8.0)
  (Transformation|SetActorLocation (Math|Vector|SelectVector (.Location (Collision|BreakHitResult _outhit)) _returnvalue_10 _returnvalue_11) false true)
  (Variables|/Script/Engine.MovementComponent|SetVelocity _returnvalue_6 (Variables|Default|GetProjectileMovement)))

(fn ConstructionScript
  (Networking|SwitchHasAuthority
    (:Authority
      (Variables|/Script/Engine.PrimitiveComponent|SetbGenerateOverlapEvents true (Variables|Default|GetExplosionTrigger))
      (Variables|/Script/Engine.PrimitiveComponent|SetbGenerateOverlapEvents true (Variables|Default|GetBeepBlinkTrigger))
      (|CallParentFunction))))
