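; Trajectory actor Blueprint for the Badger Bangs prototype weapon, by its name a projectile path preview with an attached widget.
; On BeginPlay it calls the parent, then issues a forward, endlessly looping PlayAnimation on its Widget component's user widget, with no animation asset visible in the call.
; The construction script only calls the parent, and the actor overlap event is empty, so any trajectory logic lies outside this graph.
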(event EventBeginPlay
  (|CallParentFunction)
  (UserInterface|Animation|PlayAnimation (UserInterface|GetUserWidgetObject (Variables|Default|GetWidget)) 0 0.0 0 "Forward" 1.0 false))

(event Collision|EventActorBeginOverlap
  (params (OtherActor :type Actor%20Object%20Reference)))

(fn ConstructionScript
  (|CallParentFunction))
