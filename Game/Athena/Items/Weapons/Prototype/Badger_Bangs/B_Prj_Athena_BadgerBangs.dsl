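; Thrown Badger Bangs (batarang) projectile that explodes on valid targets or sticks into the world.
; BeginPlay hides the held weapon, optionally randomizes mesh spin, and configures homing, gated by a forward-cone check and stopped by timer.
; Trigger hits, line-of-sight-checked proximity overlaps and water contact explode it with camera shake.
; Other hits spawn a gravity-aware stuck BGA on authority, except already-stuck bangs, which only spark and increase gravity.
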(event Custom|StopHoming
  ;[1 1 1] Begin Play.  Return if nothing is hit after a delay
  (Projectile|SetHomingStyle (Variables|Movement|GetProjectileMovementComponent) "None" false :!class /Script/FortniteGame.FortProjectileMovementComponent))

(event Custom|Prj_InitialRotator_Clockwise
  ;[1 1 1] Prj_InitialRotator
  (bind _result (Math|Float|ManipulateFloatInternal (Variables|Default|GetPrj_InitialRotation) (+ (Variables|Default|GetPrj_InitialRotation) (Variables|Default|GetPrj_InitialRotation_Speed)))
    (:%20%20
      (Transformation|SetRelativeRotation (Variables|Default|GetSkeletalMesh) (Math|Rotator|MakeRotator _result) false false))))

(event OnComponentActivated(ProjectileMovementComponent)
  (params (Component :type Actor%20Component%20Object%20Reference) (bReset :type Boolean))
  ;[1 1 1] AnimTrail
  (Networking|CanExecuteCosmeticEvents
    (:True
      (Effects|Particles|Trails|BeginTrails (Variables|Default|GetP_Ashton_Chicago_ShieldTrail) "ArmStrap_A" "ArmStrap_B" "ETrailWidthMode_FromCentre" 3.0))))

(event EventBeginPlay
  ;[1 1 1] Begin Play.  Return if nothing is hit after a delay
  (|CallParentFunction)
  (Variables|Default|SetisDead false)
  (Actor|SetLifeSpan 3e+01)
  (Utilities|FlowControl|Sequence
    (:then_0
      (bind _asfort_player_pawn (Utilities|Casting|CastToFortPlayerPawn (Variables|Actor|GetInstigator)))
      (bind _output_get (Variables|Default|SetOwningPlayerPawn _asfort_player_pawn))
      (Utilities|FlowControl|Sequence
        (:then_0
          (GameplayEffects|ApplyGameplayEffectToTarget :self (Ability|GetAbilitySystemComponent _output_get) :GameplayEffectClass "/Game/Athena/Items/Gameplay/BackPacks/Ashton/Chicago/GE_Ashton_Chicago_DeepDishDelivery_GE.GE_Ashton_Chicago_DeepDishDelivery_GE_C" :Target (Ability|GetAbilitySystemComponent _output_get) :Level 0.0) ; Hides projectile on arm mesh.  Needs to go here so it never gets applied until proj has spawned.\nTODO: Replace with custom GE, assess whether this is necessary with new custom animation.
          (bind _asda_badger_bangs (Utilities|Casting|CastToDA_BadgerBangs (Variables|/Script/FortniteGame.FortPawn|GetCurrentWeapon (Variables|Default|GetOwningPlayerPawn))))
          (Development|SetHiddeninGame (Variables|/Script/FortniteGame.FortWeapon|GetWeaponMesh _asda_badger_bangs) true true) ; Weapon is unhidden in the Catch ability
        )
        (:then_1
          (Audio|PlaySoundatLocation (select (Pawn|IsLocallyControlled _output_get) "/Game/Sounds/Fort_Weapons/BlackMonday/Bangs/BlackMonday_Bang_Throw_1P_Cue.BlackMonday_Bang_Throw_1P_Cue" "/Game/Sounds/Fort_Weapons/BlackMonday/Bangs/BlackMonday_Bang_Throw_Cue.BlackMonday_Bang_Throw_Cue") (Transformation|GetActorLocation)))))
    (:then_1
      (Utilities|Time|SetTimerbyEvent (delegate Custom|StopHoming) 5.0))
    (:then_2
      ;[1 1 1] Prj_InitialRotator
      (if (Variables|Default|GetModifyInitialRotation)
        (bind _output_get_2 (Variables|Default|SetPrj_InitialRotation (Math|Random|RandomFloatinRange 1.5e+02 2.1e+02)))
        (Transformation|SetRelativeRotation (Variables|Default|GetSkeletalMesh) (Math|Rotator|MakeRotator _output_get_2) false false)
        (Variables|Default|SetPrj_InitialRotation_Speed (Math|Random|RandomFloatinRange 1.0 2.8))
        (if (Math|Random|RandomBool)
          (Utilities|Time|SetTimerbyEvent (delegate Custom|Prj_InitialRotator_Clockwise) 0.02 true) ; TODO: Delete if we aren't going to adjust the projectile rotation.
          (else
            (Utilities|Time|SetTimerbyEvent (delegate Custom|Prj_InitialRotator_Anticlockwise) 0.02 true) ; TODO: Delete if we aren't going to adjust the projectile rotation.
          ))))
    (:then_3
      (Utilities|Array|AddUnique (Variables|Default|GetActorsToIgnoreForTargeting) (Variables|Default|GetOwningPlayerPawn)))
    (:then_4
      (CallFunction|SetHomingVars 0)
      (Components|Sphere|SetSphereRadius (Variables|Default|GetProximityDetection) (ScalableFloat|GetValueatLevel (Variables|Default|GetProxTriggerRadius)) true)
      (Game|BindEventtoOnDestroyed (delegate Custom|Explode)))))

(event Custom|Prj_InitialRotator_Anticlockwise
  ;[1 1 1] Prj_InitialRotator
  (bind _result (Math|Float|ManipulateFloatInternal (Variables|Default|GetPrj_InitialRotation) (+ (Variables|Default|GetPrj_InitialRotation) (* (Variables|Default|GetPrj_InitialRotation_Speed) -1.0)))
    (:%20%20
      (Transformation|SetRelativeRotation (Variables|Default|GetSkeletalMesh) (Math|Rotator|MakeRotator _result) false false))))

(event Custom|Explode
  (params (DestroyedActor :type Actor%20Object%20Reference)))

(event Custom|OnHitSomething
  (params (InputPin :type Hit%20Result%20Structure%20%28by%20ref%29))
  (bind _lasthit (Variables|Default|GetLastHit))
  (bind (_bblockinghit _binitialoverlap _time _distance _location _impactpoint _normal) (Collision|BreakHitResult _lasthit))
  (bind _projectilemovementcomponent (Variables|Movement|GetProjectileMovementComponent))
  ;[1 1 1] Event Hit: Trigger a Projectile SphereTrace.
  (if (Variables|Default|GetHasHitWater)
    (CallFunction|HitExplosionTrigger 0)
    (else
      (Variables|Default|SetLastHit InputPin)
      (Variables|Default|SetLastRotation (Transformation|GetWorldRotation (Variables|Default|GetP_BlackMonday_BatarangSpin_Mesh)))
      (Effects|Particles|Trails|EndTrails (Variables|Default|GetP_Ashton_Chicago_ShieldTrail))
      (Components|Activation|Deactivate (Variables|Default|GetP_BlackMonday_BatarangSpin_Mesh))
      (Variables|Default|SetLastVelocity (Variables|/Script/Engine.MovementComponent|GetVelocity _projectilemovementcomponent))
      (Variables|/Script/Engine.MovementComponent|SetVelocity "0, 0, 0" _projectilemovementcomponent)
      (Utilities|Time|PauseTimerbyHandle (Variables|Default|GetReturnTimer))
      (Networking|SwitchHasAuthority
        (:Authority
          (Variables|Default|SetProjectileHitNormal _normal)
          (Variables|Default|SetProjectileHitLoc _location)
          (Projectile|SetHomingTarget 0 self 0 0 :!class /Script/FortniteGame.FortProjectileMovementInterface)
          (CallFunction|HitWorld (Variables|Default|GetHitEnvSoundCue) (Variables|Default|GetProjectileHitLoc)))))))

(event OnComponentBeginOverlap(CapsuleComponent)
  (params (OverlappedComponent :type Primitive%20Component%20Object%20Reference) (OtherActor :type Actor%20Object%20Reference) (OtherComp :type Primitive%20Component%20Object%20Reference) (OtherBodyIndex :type Integer) (bFromSweep :type Boolean) (SweepResult :type Hit%20Result%20Structure%20%28by%20ref%29))
  (bind (_asfort_water_body_actor _bsuccess) (Utilities|Casting|CastToFortWaterBodyActor OtherActor))
  (bind (_asfort_athena_water _bsuccess_2) (Utilities|Casting|CastToFortAthenaWater OtherActor))
  ;[1 1 1] Event Hit: Trigger a Projectile SphereTrace.
  (if (or _bsuccess _bsuccess_2)
    (CallFunction|HitExplosionTrigger OtherActor)))

(event Custom|HitExplosionTrigger
  (params (HitActor :type Actor%20Object%20Reference))
  ;[1 1 1] Event Hit: Trigger a Projectile SphereTrace.
  (CallFunction|PlayHitPawnSounds HitActor)
  (Delivery|DoExplosion)
  (GameplayCue|ExecuteGameplayCue%28Burst%29Local self (Variables|Default|GetCameraShakeGC) (Ability|GameplayCue|MakeGameplayCueParameters :NormalizedMagnitude 1.0 :RawMagnitude 3.0 :Location (Transformation|GetActorLocation) :Normal "0,0,0" :Instigator (Variables|Actor|GetInstigator) :EffectCauser self :GameplayEffectLevel 1 :AbilityLevel 1 :bReplicateLocationWhenUsingMinimalRepProxy false))
  (Rendering|SetActorHiddenInGame true)
  (Collision|SetActorEnableCollision false)
  (Actor|SetLifeSpan 0.5)
  (Variables|Default|SetisDead true))

(event OnComponentHit(CapsuleComponent)
  (params (HitComponent :type Primitive%20Component%20Object%20Reference) (OtherActor :type Actor%20Object%20Reference) (OtherComp :type Primitive%20Component%20Object%20Reference) (NormalImpulse :type Vector :default "0, 0, 0") (Hit :type Hit%20Result%20Structure%20%28by%20ref%29))
  (bind (_bblockinghit _binitialoverlap _time _distance _location _impactpoint _normal) (Collision|BreakHitResult Hit))
  ;[1 1 1] Event Hit: Trigger a Projectile SphereTrace.
  (if (not (Variables|Default|GetisDead))
    (Utilities|Casting|CastToBGA_Athena_BadgerBangsStuck OtherActor
      (:then
        (Effects|SpawnEmitteratLocation "/Game/Athena/Items/Weapons/Prototype/Badger_Bangs/FX/P_BlackMonday_HitWorld.P_BlackMonday_HitWorld" _location (Math|Conversions|RotationFromXVector _normal))
        (Variables|/Script/Engine.ProjectileMovementComponent|SetProjectileGravityScale 5.0 (Variables|Movement|GetProjectileMovementComponent)))
      (:CastFailed
        (Utilities|CheckforExplosionTriggers OtherActor
          (:Triggered
            (CallFunction|HitExplosionTrigger OtherActor))
          (:Not%20Triggered
            (CallFunction|OnHitSomething Hit))))) ; Don't stick to other badgerbangs
  ))

(event Custom|HitWorld
  (params (Sound :type Sound%20Base%20Object%20Reference) (Location :type Vector))
  ;[1 1 1] Event Hit: Trigger a Projectile SphereTrace.
  (CallFunction|SpawnBGA (Variables|Default|GetLastHit))
  (Audio|PlaySoundatLocation Sound Location))

(event OnComponentBeginOverlap(ProximityDetection)
  (params (OverlappedComponent :type Primitive%20Component%20Object%20Reference) (OtherActor :type Actor%20Object%20Reference) (OtherComp :type Primitive%20Component%20Object%20Reference) (OtherBodyIndex :type Integer) (bFromSweep :type Boolean) (SweepResult :type Hit%20Result%20Structure%20%28by%20ref%29))
  ;[1 1 1] Event Hit: Trigger a Projectile SphereTrace.
  (if (Math|Boolean|ANDBoolean (Variables|Default|GetExplodeOnImpactPlayer) (> (Actor|GetGameTimeSinceCreation) 0.025) (not (Variables|Default|GetisDead)))
    (Utilities|CheckforExplosionTriggers OtherActor
      (:Triggered
        (Utilities|CheckLOS OtherActor
          (:LOS
            (CallFunction|HitExplosionTrigger OtherActor)))))))

(event Custom|SpawnBGA
  (params (HitResult :type Hit%20Result%20Structure%20%28by%20ref%29))
  (bind (_bblockinghit _binitialoverlap _time _distance _location _impactpoint) (Collision|BreakHitResult HitResult))
  (bind (_bblockinghit_2 _binitialoverlap_2 _time_2 _distance_2 _location_2 _impactpoint_2 _normal _impactnormal _physmat _hitactor _hitcomponent) (Collision|BreakHitResult HitResult))
  ;[1 1 1] Spawn BGA
  (Utilities|FlowControl|DoOnce false
    (:Completed
      (Utilities|FlowControl|Sequence
        (:then_0
          (Actor|SetLifeSpan 0.25)
          (Collision|SetActorEnableCollision false)
          (Rendering|SetActorHiddenInGame true))
        (:then_1
          (bind _returnvalue (Building|SpawnBuildingGameplayActor (Variables|Default|GetBGAtoSpawn) (Math|Transform|MakeTransform (- _impactpoint (* (Transformation|GetForwardVector (Variables|Default|GetP_BlackMonday_BatarangSpin_Mesh)) 2e+01)) (Transformation|GetWorldRotation (Variables|Default|GetP_BlackMonday_BatarangSpin_Mesh)) "2.000000,2.000000,2.000000") (Variables|Actor|GetInstigator) false false))
          (bind _asbga_athena_with_gravity_parent (Utilities|Casting|CastToBGA_Athena_WithGravity_Parent _returnvalue))
          (Class|BGAAthenawithGravityParent|MulticastHandleBinding _asbga_athena_with_gravity_parent _hitactor _hitcomponent)
          (Class|BGAAthenawithGravityParent|Init _asbga_athena_with_gravity_parent _impactnormal)
          (bind _asbga_athena_badger_bangs_stuck (Utilities|Casting|CastToBGA_Athena_BadgerBangsStuck _asbga_athena_with_gravity_parent))
          (Variables|/Game/Athena/Items/Weapons/Prototype/Badger_Bangs/BGA_Athena_BadgerBangsStuck.BGA_Athena_BadgerBangsStuck_C|SetImpactNormal _impactnormal _asbga_athena_badger_bangs_stuck))))))

(fn OnRep_bHit)

(fn OnRep_HitLocation)

(fn OnRep_PlayerLocation)

(fn OnRep_Target)

(fn PlayHitPawnSounds
  (params (HitActor :type Actor%20Object%20Reference))
  (bind _returnvalue (Transformation|GetActorLocation))
  (bind _owningplayerpawn (Variables|Default|GetOwningPlayerPawn))
  (Utilities|IsValid _owningplayerpawn
    (:Is%20Valid
      (if (Pawn|IsLocallyControlled _owningplayerpawn)
        (Audio|PlaySoundatLocation (Variables|Default|GetHitPlayerSoundCue) _returnvalue)
        (else
          (bind _asfort_player_pawn (Utilities|Casting|CastToFortPlayerPawn HitActor)
            (:then
              (if (Pawn|IsLocallyControlled _asfort_player_pawn)
                (Audio|PlaySoundatLocation (Variables|Default|GetHitPlayerSoundCue) _returnvalue)
                (else
                  _)))
            (:CastFailed
              _))
          (Audio|PlaySoundatLocation (Variables|Default|GetHitEnvSoundCue) _returnvalue))))))

(fn PrintActorsFromArray
  (params (ActorArray :type Array%20of%20Actor%20Object%20References))
  (for _ ActorArray
    (Development|PrintString (Utilities|String|Append "Hit;" (Utilities|GetDisplayName (Utilities|Array|Get(acopy) ActorArray (LocalVariables|GetHitActorIndex)))) true true "(R=1.000000,G=0.019947,B=0.059748,A=1.000000)" 5.0)
    (LocalVariables|SetHitActorIndex (+ (LocalVariables|GetHitActorIndex) 1))))

(fn SetHomingVars
  (params (HomingTarget :type Actor%20Object%20Reference))
  (bind _self self)
  (bind _homingtarget (Variables|Default|GetHomingTarget))
  (if (Variables|Default|GetHoming)
    (bind _output_get (Variables|Default|SetHomingTarget HomingTarget))
    (Utilities|IsValid _output_get
      (:Is%20Valid
        (Projectile|SetHomingStyle 0 0 "LockOn" false :!class /Script/FortniteGame.FortProjectileMovementInterface)
        (Utilities|Array|AddUnique (Variables|Default|GetActorsToIgnoreForTargeting))
        (Networking|SwitchHasAuthority
          (:Authority
            ;[1 1 1] TODO: Make hotfixable values.
            (Projectile|SetHomingTurnSpeedMin 0 _self 1e+01 :!class /Script/FortniteGame.FortProjectileMovementInterface)
            (Projectile|SetHomingTurnSpeedMax 0 _self 1e+02 :!class /Script/FortniteGame.FortProjectileMovementInterface)
            (Projectile|SetHomingTimeUntilMaxTurnSpeed 0 _self 0.5 :!class /Script/FortniteGame.FortProjectileMovementInterface)
            (Projectile|SetHomingTargetOffset _self (Variables|Default|GetTargetHomingOffset) :!class /Script/FortniteGame.FortProjectileMovementInterface)
            (Utilities|IsValid _homingtarget
              (:Is%20Valid
                (if (>= (Math|Vector|DotProduct (Math|Vector|Normalize (- (Transformation|GetActorLocation (Variables|Default|GetHomingTarget)) (Transformation|GetActorLocation (Variables|Default|GetOwningPlayerPawn)))) (Math|Vector|Normalize (Transformation|GetActorForwardVector (Variables|Default|GetOwningPlayerPawn)))) 0.49)
                  (Projectile|SetHomingTarget 0 self _homingtarget 0 :!class /Script/FortniteGame.FortProjectileMovementInterface))))))))))

(fn ConstructionScript
  (|CallParentFunction))
