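; Deco tool weapon for the Badger Bangs prototype, built on Ashton Chicago shield tags and audio, that relays input to gameplay abilities.
; Primary press and secondary press/release send gameplay events to the instigator; primary release does nothing.
; On equip it caches the player pawn, then hides the mesh if the shield-out tag matches, otherwise plays the equip sound and unhides it.
; Equip and local unequip toggle the IsAshtonChicago flag.
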(event Custom|UnhideWeaponMeshAll
  (Development|SetHiddeninGame (Variables|Visual|GetWeaponMesh) false true))

(event FortWeapon|EventOnEquippedAndAttached
  (bind _weaponmesh (Variables|Visual|GetWeaponMesh))
  ;[1 1 1] Unhide the actor if it's hidden, play equip sound if it's not thrown
  (Variables|Default|SetIsAshtonChicago true)
  (bind _asfort_player_pawn (Utilities|Casting|CastToFortPlayerPawn (Game|GetInstigator)))
  (bind _output_get (Variables|Default|SetPlayerPawn _asfort_player_pawn))
  (bind _returnvalue (GameplayTags|HasAnyMatchingGameplayTags(Message) _output_get (Variables|Default|GetShieldOutTag)))
  (if _returnvalue
    (Development|SetHiddeninGame (Variables|Visual|GetWeaponMesh) true false)
    (else
      (Audio|PlaySoundatLocation "/Game/Athena/Sounds/LTMs/Ashton/Chicago/Ashton_Chicago_Equip_Cue.Ashton_Chicago_Equip_Cue" (Transformation|GetActorLocation _output_get))
      (if (Variables|/Script/Engine.SceneComponent|GetbHiddenInGame _weaponmesh)
        (Development|SetHiddeninGame _weaponmesh false false)))))

(event FortWeapon|EventOnLocalUnEquip
  ;[1 1 1] Unhide the actor if it's hidden, play equip sound if it's not thrown
  (Variables|Default|SetIsAshtonChicago false))

(event DecoTool|EventOnPressSecondaryFire
  (params (FortDecoHelper :type Fort%20Deco%20Helper%20Object%20Reference))
  ;[1 1 1] Ability handler. This deco tool acts as an event manager for the abilities running when the item is picked up.
  (Ability|SendGameplayEventtoActor (Game|GetInstigator) (Variables|Default|GetSecondaryFireEventTag)))

(event DecoTool|EventOnReleaseSecondaryFire
  (params (FortDecoHelper :type Fort%20Deco%20Helper%20Object%20Reference))
  ;[1 1 1] Ability handler. This deco tool acts as an event manager for the abilities running when the item is picked up.
  (Ability|SendGameplayEventtoActor (Game|GetInstigator) "(TagName=\"Event.Athena.Item.Ashton.Chicago.Defense.End\")"))

(event DecoTool|EventOnPressTrigger
  (params (FortDecoHelper :type Fort%20Deco%20Helper%20Object%20Reference))
  ;[1 1 1] Ability handler. This deco tool acts as an event manager for the abilities running when the item is picked up.
  ;[0.3904 0.4224 1] EVENT -> SecondaryFire
  (Ability|SendGameplayEventtoActor (Game|GetInstigator) (Variables|Default|GetPrimaryFireEventTag)))

(event DecoTool|EventOnReleaseTrigger
  (params (FortDecoHelper :type Fort%20Deco%20Helper%20Object%20Reference)))

(fn ConstructionScript)
