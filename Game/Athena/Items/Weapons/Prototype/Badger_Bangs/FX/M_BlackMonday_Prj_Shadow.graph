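# Unlit, translucent, two-sided particle material for the shadow effect of the BlackMonday (Badger_Bangs) projectile, usable on sprite and mesh particles.
# Emissive color is a squared soft radial mask times particle color applied twice, passed through MF_FX_EmissiveAdjustments.
# A second chain spins the projectile texture rapidly, offset by a dynamic parameter, then selects a channel (blue by default) and applies an adjustable power.
# That shape is boosted, sphere-masked and scaled by particle alpha, apparently for opacity.
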Material /Game/Athena/Items/Weapons/Prototype/Badger_Bangs/FX/M_BlackMonday_Prj_Shadow
graph {
 properties {
  bUsedWithMeshParticles=true
  bUsedWithNiagaraMeshParticles=true
  bUsedWithNiagaraSprites=true
  bUsedWithParticleSprites=true
  blendMode=BLEND_Translucent
  pixelDepthOffsetMode=PDOM_Legacy
  shadingModel=MSM_Unlit
  twoSided=true
  usedShadingModels=Unlit
 }
 nodes [
  n1 = SphereMask(A=TextureCoordinate(), B=Constant() {r=0.5}) {attenuationRadius=0.5, hardnessPercent=0}
  n2 = Multiply(A=@n1, B=@n1)
  n3 = ParticleColor()
  MaterialFunctionCall(Emissive=Multiply(A=Multiply(A=@n2, B=@n3.RGB), B=@n3.RGB)) {materialFunction=&MF_FX_EmissiveAdjustments}
  n4 = Rotator(Time=Add(A=Time(), B=DynamicParameter().Param1)) {speed=-40}
  n5 = Power(
   Base=StaticComponentMaskParameter(None=TextureSample(UVs=@n4) {texture=&/T_BlackMonday_Prj}.RGB) {defaultB=true, parameterName=Channel}
   Exp=ScalarParameter() {defaultValue=1, parameterName=Power}
  )
  Multiply(
   A=Multiply(A=Multiply(A=@n5) {constB=1.5}, B=TextureSample() {texture=/Game/Effects/Fort_Effects/Textures/Power_Source/T_Power_Source_Sphere_Mask}.R)
   B=@n3.A
  )
 ]
}
