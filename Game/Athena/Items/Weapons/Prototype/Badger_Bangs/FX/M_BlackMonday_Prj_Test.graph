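# Lit mesh-particle material, named as a test for the Black Monday Bangs projectile, with a red emissive glow.
# Vertex red blends dark and light greys and a pair of scalar values.
# Two world-space sphere masks around the particle position build a soft grey-to-red halo plus a tight, bright red core.
# A shading path switch uses a red-tinted emissive texture on mobile instead, feeding MF_FX_EmissiveAdjustments.
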Material /Game/Athena/Items/Weapons/Prototype/Badger_Bangs/FX/M_BlackMonday_Prj_Test
graph {
 properties {bUsedWithMeshParticles=true, pixelDepthOffsetMode=PDOM_Legacy, usedShadingModels=Default Lit}
 nodes [
  Constant() {r=1}
  n1 = VertexColor()
  LinearInterpolate(
   A=VectorParameter() {defaultValue=<rgba 0.0469 0.0469 0.0469 0>}.RGB
   B=VectorParameter() {parameterName=Param1, defaultValue=<rgba 0.8542 0.8542 0.8542 0>}.RGB
   Alpha=@n1.R
  )
  LinearInterpolate(Alpha=@n1.R) {constA=0.4, constB=0.15}
  n2 = SphereMask(A=WorldPosition().XYZ, B=ParticlePositionWS()) {attenuationRadius=38, hardnessPercent=0}
  n3 = Multiply(
   A=LinearInterpolate(A=Constant3Vector() {constant=<rgba 0.063 0.0776 0.0885 0>}, B=Constant3Vector() {constant=<rgba 0.0833 0 0 0>}, Alpha=Power(Base=@n2))
   B=@n2
  )
  n4 = Multiply(A=SphereMask(A=WorldPosition().XYZ, B=ParticlePositionWS()) {attenuationRadius=4.5}, B=Constant3Vector() {constant=<rgba 15 0 0 0>})
  n5 = Constant3Vector() {constant=<rgba 5 0 0 0>}
  n6 = TextureSample() {texture=/Game/Gadgets/Assets/BlackMonday_Bangs/T_BlackMonday_Bangs_E}
  MaterialFunctionCall(Emissive=ShadingPathSwitch(Default=Add(A=@n3, B=@n4), Mobile=Multiply(A=@n5, B=@n6.R)) {desc=Migrated from FeatureLevelSwitch}) {materialFunction=&MF_FX_EmissiveAdjustments}
 ]
}
