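# Lit material for the BlackMonday Bangs prototype weapon, previewed on its projectile mesh, with diffuse, mask and normal textures.
# Builds a procedural vertex wobble: a time-driven sine rotates world position slightly around local Z at a pivot offset from the object origin.
# The wobble fades linearly over AnimDuration from the GameTime timestamp parameter.
# The graph shows no material output links, and the MeshPivot transform feeds nothing.
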Material /Game/Athena/Items/Weapons/Prototype/Badger_Bangs/FX/M_BlackMonday_Wave
graph {
 properties {
  bUsedWithMeshDeformer=true
  bUsedWithSkeletalMesh=true
  pixelDepthOffsetMode=PDOM_Legacy
  previewMesh=&/SM_BlackMonday_Prj
  usedShadingModels=Default Lit
 }
 nodes [
  Constant() {r=1}
  Constant() {r=0.4}
  n1 = Sine(None=Multiply(A=Time()) {constB=4.5}) {period=0.2}
  n2 = Subtract(A=Time(), B=ScalarParameter() {defaultValue=1, desc=Time-stamp, parameterName=GameTime})
  n3 = ScalarParameter() {defaultValue=0.5, parameterName=AnimDuration}
  n4 = OneMinus(None=Saturate(None=Divide(A=@n2, B=@n3)))
  n5 = ComponentMask(None=VectorParameter() {defaultValue=<rgba 14 0 0 1>}.RGB) {b=true, g=true, r=true}
  RotateAboutAxis(
   NormalizedRotationAxis=Transform(None=Constant3Vector() {constant=<rgba 0 0 1 0>}) {transformSourceType=TRANSFORMSOURCE_Local}
   RotationAngle=Multiply(A=Multiply(A=@n1, B=@n4) {constB=0.07}) {constB=0.0063}
   PivotPoint=Add(A=ObjectPositionWS(), B=Transform(None=@n5) {transformSourceType=TRANSFORMSOURCE_Local})
   Position=WorldPosition().XYZ
  )
  TextureSample() {texture=/Game/Gadgets/Assets/BlackMonday_Bangs/Textures/T_BlackMonday_Bangs_D}
  TextureSample() {samplerType=SAMPLERTYPE_Masks, texture=/Game/Gadgets/Assets/BlackMonday_Bangs/Textures/T_BlackMonday_Bangs_S}
  TextureSample() {samplerType=SAMPLERTYPE_Normal, texture=/Game/Gadgets/Assets/BlackMonday_Bangs/Textures/T_BlackMonday_Bangs_N}
  Transform(None=VectorParameter() {parameterName=MeshPivot, defaultValue=<rgba 1 1 1 1>}.RGB) {transformSourceType=TRANSFORMSOURCE_World, transformType=TRANSFORM_ParticleWorld}
 ]
}
