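# Niagara system for a looping alert glow on the BlackMonday prototype weapon, red-orange by default.
# One emitter spawns camera-offset sprites at a user rate, rendered by four sprite renderers and a light.
# User Static and Blink bools mask two sets of color and size attributes, with cycling curves pulsing the Blink set.
# Spawned particles export data to Blueprint through User.BPData; a second Static emitter has no renderers.
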NiagaraSystem /Game/Athena/Items/Weapons/Prototype/Badger_Bangs/FX/NS_BlackMonday_AlertLoop
dependencies {
 usedDataInterfaces [/Script/Niagara.NiagaraDataInterfaceExport, &NiagaraDataInterfaceCurve]
 usedDynamicInputs []
 usedModules [
  &SystemState
  &self:SystemSpawnScript.SystemScriptSource.SystemScriptGraph.NiagaraNodeAssignment_0.SetVariables_#1
 ]
}
emitters {
 SingleLoopingParticle {
  emitterUpdateScript [
   &EmitterState()
   &SpawnBurst_Instantaneous(Age:NiagaraFloat=$System.LoopedAge, Use Spawn Probability=false, UseLoopCountLimit=false) {enabled=false}
   &SpawnRate(SpawnGroup=1, SpawnRate=$User.Spawn Rate, Use Spawn Probability=false)
  ]
  particleSpawnScript [
   &InitializeParticle(
    Color=$User.Color
    Lifetime=&MaskFloatBySpawnGroup(Masked Float=1.5, Passthrough Float=$User.Lifetime)
    Sprite Size=<300 160>
    Sprite Size Mode=::NewEnumerator4
    UsePositionOffset=false
    Write Material Random=true
   )
   /SetVariables(Particles.VisibilityTag=&SpawnGroup()) {isSetParameters=true}
   /Niagara/Modules/ExportParticleData/ExportParticleDataToBlueprint(
    Condition To Export Data=true
    Delay Between Data Exports=0.01
    Export Particle Data Interface=DataInterface() {propertyValues {
      GPUAllocationFixedSize=64
      GPUAllocationMode=FixedSize
      GPUAllocationPerParticleSize=1
      CallbackHandlerParameter {parameter {name=User.BPData, typeDefHandle {registeredTypeIndex=1}, varData []}}
     }}
    "Float To Send (As Struct Size)"=1
    UseDelayTimer=true
   )
   &CameraOffset(Camera Offset Amount=10)
  ]
  particleUpdateScript [
   &ParticleState()
   &ScaleColor(
    Scale RGB=&VectorFromFloat(Value=&FloatFromCurve(
      FloatCurve=DataInterface() {propertyValues {Curve {
         postInfinityExtrap=RCCE_Cycle
         keys [
          {arriveTangent=-0.4914, interpMode=RCIM_Cubic, leaveTangent=-0.4914, tangentMode=RCTM_User, time=0, value=1}
          {arriveTangent=0.0954, interpMode=RCIM_Cubic, leaveTangent=0.0954, time=0.2895, value=0.8706}
          {arriveTangent=-0.0385, interpMode=RCIM_Cubic, leaveTangent=-0.0385, tangentMode=RCTM_User, time=0.7864, value=1.075}
          {arriveTangent=-0.5544, interpMode=RCIM_Cubic, leaveTangent=-0.5544, tangentMode=RCTM_User, time=1, value=1}
         ]
        }}}
      CurveIndex=$Emitter.Age
     ))
    ScaleA=false
    ScaleRGB=true
   ) {module=ScaleColor001}
   &SolveForcesAndVelocity(Clamp Velocity=false, Limit Acceleration=false)
   /SetVariables(
    Particles.Color_A:LinearColor=&Multiply_LinearColorByFloat(
     Linear Color=$Particles.Color
     Float=&Multiply_Float(A=&MakeCustomFloatFromBool(Bool=$User.Static, True Float=1, False Float=0), B=1)
    )
    Particles.GlowSize_A:Vector2f=&Multiply_Vector2D_ByFloat(
     Float=&Multiply_Float(A=&MakeCustomFloatFromBool(Bool=$User.Static, True Float=1, False Float=0), B=0.5)
     Vector2D=$Particles.SpriteSize
    )
    Particles.SpriteSize_A:Vector2f=&Multiply_Vector2D_ByFloat(
     Float=&Multiply_Float(
      A=&MakeCustomFloatFromBool(Bool=$User.Static, True Float=1, False Float=0)
      B=&FloatFromCurve(
       FloatCurve=DataInterface() {propertyValues {Curve {
          defaultValue=1
          keys [
           {arriveTangent=7.2082, interpMode=RCIM_Cubic, leaveTangent=7.2082, tangentMode=RCTM_User, time=0, value=1}
           {arriveTangent=-2.1539, interpMode=RCIM_Cubic, leaveTangent=-2.1539, tangentMode=RCTM_User, time=0.1, value=3}
           {arriveTangent=0.0427, interpMode=RCIM_Cubic, leaveTangent=0.0427, tangentMode=RCTM_User, time=0.3, value=1}
          ]
         }}}
       CurveIndex=$Emitter.Age
      )
     )
     Vector2D=$Particles.SpriteSize
    )
   ) {isSetParameters=true}
   /SetVariables(
    Particles.Color_B:LinearColor=&Multiply_LinearColorByFloat(
     Linear Color=$Particles.Color
     Float=&Multiply_Float(
      A=&FloatFromCurve(
       FloatCurve=DataInterface() {propertyValues {Curve {
          postInfinityExtrap=RCCE_Cycle
          keys [
           {arriveTangent=9.0725, interpMode=RCIM_Cubic, leaveTangent=9.0725, tangentMode=RCTM_User, time=0, value=0.4}
           {arriveTangent=-0.2093, interpMode=RCIM_Cubic, leaveTangent=-0.2093, tangentMode=RCTM_User, time=0.1, value=0.9}
           {arriveTangent=-0.5556, interpMode=RCIM_Cubic, leaveTangent=-0.5556, time=0.4, value=0.65}
           {arriveTangent=-0.0499, interpMode=RCIM_Cubic, leaveTangent=-0.0499, tangentMode=RCTM_User, time=1, value=0.4}
          ]
         }}}
       Scale Curve=1.5
      )
      B=&Multiply_Float(A=&MakeCustomFloatFromBool(Bool=$User.Blink, True Float=1, False Float=0), B=1)
     )
    )
    Particles.GlowSize_B:Vector2f=&Multiply_Vector2D_ByFloat(
     Float=&Multiply_Float(A=&MakeCustomFloatFromBool(Bool=$User.Blink, True Float=1, False Float=0), B=0.5)
     Vector2D=$Particles.SpriteSize
    )
    Particles.SpriteSize_B:Vector2f=&Multiply_Vector2D_ByFloat(
     Float=&Multiply_Float(
      A=&FloatFromCurve(
       FloatCurve=DataInterface() {propertyValues {Curve {
          postInfinityExtrap=RCCE_Cycle
          keys [
           {arriveTangent=9.0725, interpMode=RCIM_Cubic, leaveTangent=9.0725, tangentMode=RCTM_User, time=0, value=0.35}
           {arriveTangent=-0.2093, interpMode=RCIM_Cubic, leaveTangent=-0.2093, tangentMode=RCTM_User, time=0.1, value=0.9}
           {arriveTangent=-0.6111, interpMode=RCIM_Cubic, leaveTangent=-0.6111, time=0.4, value=0.45}
           {arriveTangent=-0.0499, interpMode=RCIM_Cubic, leaveTangent=-0.0499, tangentMode=RCTM_User, time=1, value=0.35}
          ]
         }}}
       Scale Curve=5
      )
      B=&Multiply_Float(A=&MakeCustomFloatFromBool(Bool=$User.Blink, True Float=1, False Float=0), B=1)
     )
     Vector2D=<125 125>
    )
   ) {isSetParameters=true}
   /SetVariables(Particles.Color_C:LinearColor=&LerpLinearColors(
     StartColor=$Particles.Color_A
     EndColor=$Particles.Color_B
     LerpFactor=&MakeCustomFloatFromBool(Bool=$User.Blink, True Float=1, False Float=0)
    )) {isSetParameters=true}
  ]
  rendererClasses [&NiagaraSpriteRendererProperties, &NiagaraLightRendererProperties]
  renderers [
   {rendererClass=&NiagaraSpriteRendererProperties}
   {rendererClass=&NiagaraSpriteRendererProperties, rendererIndex=1}
   {rendererClass=&NiagaraSpriteRendererProperties, rendererIndex=2}
   {rendererClass=&NiagaraSpriteRendererProperties, rendererIndex=3}
   {rendererClass=&NiagaraLightRendererProperties, rendererIndex=4}
  ]
 }
 Static {rendererClasses [], renderers []}
}
summary {
 systemName=NS_BlackMonday_AlertLoop
 emitters [
  {
   emitterName=SingleLoopingParticle
   rendererClasses [&NiagaraSpriteRendererProperties, &NiagaraLightRendererProperties]
  }
  {emitterName=Static, rendererClasses []}
 ]
 userVariables [
  User.Blink {type=NiagaraBool}
  User.BPData {
   type=Object
   description=""
   defaultValue {struct=NiagaraExt_VariableValue_Object, value {object=null, objectClass=&Object}}
  }
  User.Color {type=LinearColor, description="", defaultValue {struct=LinearColor, value=<rgba 1.5 0.025 0 1>}}
  User.Lifetime {type=NiagaraFloat, defaultValue=0.15, description=""}
  User.Spawn Rate {type=NiagaraFloat, defaultValue=2, description=""}
  User.Static {type=NiagaraBool}
 ]
}
