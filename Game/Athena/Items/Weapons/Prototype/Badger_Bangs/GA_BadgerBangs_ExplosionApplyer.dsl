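; Gameplay ability that applies explosion effects for the prototype Badger Bangs weapon's stuck actor.
; On activation it stores a source transform at the avatar's location, nudged outward along the stuck actor's impact normal.
; GetCustomAbilitySourceTransform returns that transform as the custom ability source.
; Target selection runs with the generic projectile Explode tag, applies the gameplay effect container to targets, then ends the ability.
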(event Ability|EventActivateAbility
  (bind _returnvalue (Ability|GetAvatarActorfromActorInfo))
  (bind (_returnvalue_location _returnvalue_rotation _returnvalue_scale) (Math|Transform|BreakTransform (Transformation|GetActorTransform _returnvalue)))
  (bind _asbga_athena_badger_bangs_stuck (Utilities|Casting|CastToBGA_Athena_BadgerBangsStuck _returnvalue))
  (Variables|SetSource_Location (Math|Transform|MakeTransform (+ (* (Variables|/Game/Athena/Items/Weapons/Prototype/Badger_Bangs/BGA_Athena_BadgerBangsStuck.BGA_Athena_BadgerBangsStuck_C|GetImpactNormal _asbga_athena_badger_bangs_stuck) 3e+01) _returnvalue_location) _returnvalue_rotation _returnvalue_scale))
  (bind (_asynctaskproxy _targetdata _applicationtag) (Ability|Tasks|DoTargetSelection "None" "(TagName=\"Abilities.Generic.GenericProjectile.Explode\")" true)
    (:Targeted
      (Ability|ApplyGameplayEffectContainer _targetdata _applicationtag)
      (Ability|EndAbility))))

(fn GetCustomAbilitySourceTransform
  (return (Variables|Default|GetSource)))
